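; Building game actor that plays a particle effect and sound when any building actor in its configured list dies.
; On BeginPlay it makes itself invulnerable, then, on authority only, binds to OnDied of each valid listed actor that can be damaged.
; Each death passes the actor's centroid to PlayEffects, which spawns the offset Niagara system and sound only on the Remote branch.
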(event EventBeginPlay
  ; Set Invulnerable
  (Variables|Actor|SetbCanBeDamaged false)
  (Variables|BuildingActor|SetbIsInvulnerable true)
  (Networking|SwitchHasAuthority
    (:Authority
      (CallFunction|BindtoPropsinList))))

(event Custom|BindToPropsInList
  ; Bind Props
  (Networking|SwitchHasAuthority
    (:Authority
      (for _array_element (Variables|Default|GetActorsToPlayEffectsOnDestroy)
        (Utilities|IsValid _array_element
          (:Is%20Valid
            (bind _asbuilding_actor (Utilities|Casting|CastToBuildingActor _array_element))
            (if (Variables|/Script/Engine.Actor|GetbCanBeDamaged _asbuilding_actor)
              (if (not (Variables|/Script/FortniteGame.BuildingActor|GetbIsInvulnerable _asbuilding_actor))
                (EventDispatchers|BindEventtoOnDied _asbuilding_actor (delegate Custom|Died))))))))))

(event Custom|Died
  (params (DamagedActor :type Actor%20Object%20Reference) (Damage :type Float%20%28single-precision%29) (InstigatedBy :type Controller%20Object%20Reference) (DamageCauser :type Actor%20Object%20Reference) (HitLocation :type Vector) (FHitComponent :type Primitive%20Component%20Object%20Reference) (BoneName :type Name) (Momentum :type Vector))
  ; Bind Props
  (bind _asbuilding_actor (Utilities|Casting|CastToBuildingActor DamagedActor))
  (CallFunction|PlayEffects (Building|GetCentroid _asbuilding_actor)))

(event Custom|PlayEffects
  (params (Location :type Vector))
  ; Play FX on died
  (Networking|SwitchHasAuthority
    (:Remote
      (Niagara|SpawnSystematLocation (Variables|Default|GetParticleSystem) (+ Location (Variables|Default|GetEffectsOffset)))
      (Audio|SpawnSoundatLocation (Variables|Default|GetSoundCue) Location))))

(fn ConstructionScript)
